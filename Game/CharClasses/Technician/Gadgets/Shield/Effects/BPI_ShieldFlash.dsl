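; Blueprint interface for the Technician class's shield gadget effects.
; It declares a single function, FlashShield, with no parameters or return values.
; The file contains no implementation, so any flash behavior is left to the classes that implement the interface.
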(fn FlashShield)
